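# Physics asset for the CordSyrup two-handed pickaxe in BRCosmetics, defining ragdoll-style bodies for a four-link dangling chain on the axe.
# A long capsule covers the handle bone, and small capsules sit on each dyn_chain bone.
# Constraints link the chain in series from the handle.
# The first link swings asymmetrically with no twist range, while the later links share wider swing and limited twist.
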PhysicsAsset /BRCosmetics/Weapons/FORT_Melee/Pickaxe_CordSyrup/Meshes/CordSyrup_Axe_PhysicsAsset
bodies {
 Melee_TwoHanded_Handle [{shapeType=Capsule, length=90.8554, radius=2.6685, center=<0.5883 -1.6505 19.0531>, rotation=<rot 0 -90 0>}]
 dyn_chain_01 [{shapeType=Capsule, length=2, radius=2, shapeName=dyn_chain_01_capsule, center=<3.5556 -0.0199 0.0149>, rotation=<rot 0 89.4657 -89.9231>}]
 dyn_chain_02 [{shapeType=Capsule, length=2, radius=2, shapeName=dyn_chain_02_capsule, center=<3.4954 -0.1452 -0.0138>, rotation=<rot 0 90.3896 -89.7959>}]
 dyn_chain_03 [{shapeType=Capsule, length=2, radius=2, shapeName=dyn_chain_03_capsule, center=<5.4939 0 0>, rotation=<rot 90 0 0>}]
 dyn_chain_04 [{shapeType=Capsule, length=2, radius=2, shapeName=dyn_chain_04_capsule, center=<5 0 0>, rotation=<rot -90 0 0>}]
}
constraints [
 {bone1Name=dyn_chain_01, bone2Name=Melee_TwoHanded_Handle, swing1LimitDegrees=10, swing2LimitDegrees=60, twistLimitDegrees=0}
 {bone1Name=dyn_chain_02, bone2Name=dyn_chain_01, swing1LimitDegrees=75, swing2LimitDegrees=75, twistLimitDegrees=15}
 {bone1Name=dyn_chain_03, bone2Name=dyn_chain_02, swing1LimitDegrees=75, swing2LimitDegrees=75, twistLimitDegrees=15}
 {bone1Name=dyn_chain_04, bone2Name=dyn_chain_03, swing1LimitDegrees=75, swing2LimitDegrees=75, twistLimitDegrees=15}
]
