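# Animation Blueprint for the CordSyrup pickaxe that adds physics-driven secondary motion to its skeleton.
# Two RigidBody nodes run in series on the handle bone: the FE physics asset simulates in world space, then the standard asset in base-bone space.
# Both use overridden gravity, damped component motion and reference-pose-limited translation.
# The event graph stores a frontend-display check as two opposite booleans, with no visible link to the RigidBody nodes.
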AnimBlueprint /BRCosmetics/Weapons/FORT_Melee/Pickaxe_CordSyrup/Meshes/CordSyrup_Axe_Skeleton_AnimBP
target_skeleton=&/CordSyrup_Axe_Skeleton
graphs [
 AnimGraph {nodes [
   n1 = RigidBody(ComponentPose=Local To Component(LocalPose=Input Pose())) {
    bClampLinearTranslationLimitToRefPose=true
    bOverrideWorldGravity=true
    overridePhysicsAsset=&/CordSyrup_Axe__FE_PhysicsAsset
    simulationSpace=WorldSpace
    alphaBoolBlend {blendOption=Linear}
    baseBoneRef {boneName=Melee_TwoHanded_Handle}
    componentAppliedLinearAccClamp=<3000 3000 3000>
    componentLinearAccScale=<0.4 0.3 0.3>
    componentLinearVelScale=<0.4 0.3 0.3>
    overrideWorldGravity {z=-980}
   }
   n2 = RigidBody(ComponentPose=@n1) {
    bClampLinearTranslationLimitToRefPose=true
    bOverrideWorldGravity=true
    overridePhysicsAsset=&/CordSyrup_Axe_PhysicsAsset
    simulationSpace=BaseBoneSpace
    alphaBoolBlend {blendOption=Linear}
    baseBoneRef {boneName=Melee_TwoHanded_Handle}
    componentAppliedLinearAccClamp=<3000 3000 3000>
    componentLinearAccScale=<0.4 0.3 0.3>
    componentLinearVelScale=<0.4 0.3 0.3>
    overrideWorldGravity {z=-980}
   }
   Output Pose(Result=Component To Local(ComponentPose=@n2))
  ]}
 EventGraph {nodes [
   Event BlueprintBeginPlay() {is=title K2Node_Event}
   n1 = IsCosmeticObjectDisplayedInFrontendContext!(CosmeticObject:Object Reference) -> (ReturnValue:bool)
   > Set IsDisplayFrontEnd(IsDisplayFrontEnd:bool=@n1) -> (Output_Get:bool)
   > Set IsNotFrontEnd(IsNotFrontEnd:bool=NOT Boolean(A=@n1)) -> (Output_Get:bool)
   TryGetPawnOwner()
   Event BlueprintUpdateAnimation()
  ]}
]
variables [
 IsDisplayFrontEnd {type=boolean}
 IsNotFrontEnd {type=boolean}
]
